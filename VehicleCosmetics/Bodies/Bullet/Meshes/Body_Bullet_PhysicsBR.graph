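# Physics asset for the Bullet vehicle body mesh, in the vehicle cosmetics content, with a BR suffix in its name.
# It holds a single body on the root bone and no constraints, so the mesh is simulated as one rigid piece with no jointed parts.
# The root body lists no collision shapes or overrides here.
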PhysicsAsset /VehicleCosmetics/Bodies/Bullet/Meshes/Body_Bullet_PhysicsBR
bodies {root []}
constraints []
